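# Physics asset for the ArcticBreeze medium male body cosmetic, giving its dangling radio and front and back belt straps physical motion.
# Capsule chains run from master bones through each strap and into the radio, which ends in a box body.
# Both strap ends are constrained to that box, linking the three chains with tight swing and twist limits.
# Unconstrained pelvis, lower spine and left thigh capsules provide body collision shapes.
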PhysicsAsset /BRCosmetics/Characters/Player/Male/Medium/Bodies/M_MED_ArcticBreeze/Meshes/M_MED_ArcticBreeze_Physics
bodies {
 dyn_backBelt_01 [{
   shapeType=Capsule
   length=3.0207
   radius=1.5345
   shapeName=dyn_backBelt_01_capsule
   center=<2.9156 -0.5157 0.2621>
   rotation=<rot -8.4897 -90 -93.5829>
  }]
 dyn_backBelt_02 [{
   shapeType=Capsule
   length=3.3065
   radius=2.0964
   shapeName=dyn_backBelt_02_capsule
   center=<2.5087 -0.406 0.0609>
   rotation=<rot 0 -77.2334 -85.3673>
  }]
 dyn_backBelt_03 [{
   shapeType=Capsule
   length=2.7224
   radius=1.7516
   shapeName=dyn_backBelt_03_capsule
   center=<2.1234 0.4004 0.4377>
   rotation=<rot -2.5337 -80.722 78.6151>
  }]
 dyn_backBelt_master [{shapeType=Capsule, length=0.505, radius=0.505, shapeName=dyn_backBelt_master_capsule, rotation=<rot 0 -90 0>}]
 dyn_frontBelt_01 [{
   shapeType=Capsule
   length=2.4531
   radius=1.7846
   shapeName=dyn_frontBelt_01_capsule
   center=<3.381 0.9041 0.6064>
   rotation=<rot -9.8094 -90 -85.1995>
  }]
 dyn_frontBelt_02 [{
   shapeType=Capsule
   length=3.8857
   radius=1.8984
   shapeName=dyn_frontBelt_02_capsule
   center=<2.5067 0.3022 0.1456>
   rotation=<rot 0 90.6207 -89.939>
  }]
 dyn_frontBelt_03 [{
   shapeType=Capsule
   length=0.505
   radius=2.2719
   shapeName=dyn_frontBelt_03_capsule
   center=<2.2099 -0.0524 0.1099>
   rotation=<rot -0.3327 -96.7028 87.1722>
  }]
 dyn_frontBelt_master [{shapeType=Capsule, length=0.505, radius=0.505, shapeName=dyn_frontBelt_master_capsule, rotation=<rot 0 -90 0>}]
 dyn_radio_01 [{shapeType=Capsule, length=10, radius=4, shapeName=dyn_radio_01_capsule, center=<0 1.9316 0>, rotation=<rot 0 0 -90>}]
 dyn_radio_02 [{
   shapeType=Box
   extentX=7.5
   extentY=11.8187
   extentZ=20.8585
   shapeName=dyn_radio_02_box
   center=<0.6986 -0.5732 0.898>
   rotation=<rot 31.4621 -5.2076 -2.7236>
  }]
 dyn_radio_master [{shapeType=Capsule, length=0.505, radius=0.505, shapeName=dyn_radio_master_capsule, rotation=<rot 0 -90 0>}]
 pelvis [
  {shapeType=Capsule, length=8.7187, radius=8.3331, shapeName=pelvis_capsule, rotation=<rot 0 -90 0>}
  {shapeType=Capsule, length=8.7187, radius=8.3331, shapeName=pelvis_capsule, center=<0.1782 -2.6703 0>, rotation=<rot 0 -90 0>}
 ]
 spine_01 [{shapeType=Capsule, length=8.225, radius=9.2457, shapeName=spine_01_capsule, center=<0 0 0.2161>, rotation=<rot 0 -90 0>}]
 spine_02 [{shapeType=Capsule, length=7.2116, radius=9.7057, shapeName=spine_02_capsule, center=<0 0 0.3257>, rotation=<rot 0.0034 -90 -0.0049>}]
 thigh_l [{
   shapeType=Capsule
   length=30.6969
   radius=7.8913
   shapeName=thigh_l_capsule
   center=<-15.9897 0.0002 0.3344>
   rotation=<rot -0.0001 89.9993 -89.3874>
  }]
}
constraints [
 {bone1Name=dyn_radio_02, bone2Name=dyn_radio_01, swing1LimitDegrees=15, swing2LimitDegrees=15, twistLimitDegrees=5}
 {bone1Name=dyn_radio_01, bone2Name=dyn_radio_master, swing1LimitDegrees=15, swing2LimitDegrees=15, twistLimitDegrees=3}
 {bone1Name=dyn_frontBelt_01, bone2Name=dyn_frontBelt_master, swing1LimitDegrees=15, swing2LimitDegrees=15, twistLimitDegrees=5}
 {bone1Name=dyn_frontBelt_02, bone2Name=dyn_frontBelt_01, swing1LimitDegrees=30, swing2LimitDegrees=30, twistLimitDegrees=10}
 {bone1Name=dyn_frontBelt_03, bone2Name=dyn_frontBelt_02, swing1LimitDegrees=30, swing2LimitDegrees=30, twistLimitDegrees=10}
 {bone1Name=dyn_backBelt_01, bone2Name=dyn_backBelt_master, swing1LimitDegrees=15, swing2LimitDegrees=15, twistLimitDegrees=5}
 {bone1Name=dyn_backBelt_02, bone2Name=dyn_backBelt_01, swing1LimitDegrees=15, swing2LimitDegrees=15, twistLimitDegrees=5}
 {bone1Name=dyn_backBelt_03, bone2Name=dyn_backBelt_02, swing1LimitDegrees=30, swing2LimitDegrees=30, twistLimitDegrees=10}
 {bone1Name=dyn_frontBelt_03, bone2Name=dyn_radio_02, swing1LimitDegrees=15, swing2LimitDegrees=15, twistLimitDegrees=5}
 {bone1Name=dyn_backBelt_03, bone2Name=dyn_radio_02, swing1LimitDegrees=15, swing2LimitDegrees=15, twistLimitDegrees=5}
]
