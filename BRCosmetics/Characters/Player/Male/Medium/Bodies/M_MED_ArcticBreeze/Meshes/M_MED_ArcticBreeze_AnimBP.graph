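# Post-copy AnimBP for the ArcticBreeze medium male body cosmetic, adding secondary motion to gear.
# Copies the parent mesh pose, offsets attach bones, runs a DBNO-aware Control Rig, then drives gear bones by weighted multi-bone constraints into a cached pose.
# That feeds two bool-blended AnimDynamics antenna chains, then pelvis-space RigidBody physics and two radio-wire AnimDynamics passes.
# The event graph updates front-end state, curve-driven values and a deprecated gravity override.
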AnimBlueprint /BRCosmetics/Characters/Player/Male/Medium/Bodies/M_MED_ArcticBreeze/Meshes/M_MED_ArcticBreeze_AnimBP
target_skeleton=&/M_MED_ArcticBreeze_Skeleton
graphs [
 AnimGraph {nodes [
   n1 = Anim Dynamics - Bone: dyn_antena_01(ComponentPose=Local To Component(LocalPose=Use cached pose 'Init'()), Alpha=1, ExternalForce="0,0,0") {
    angularDampingOverride=1
    angularSpringConstant=3000
    bAngularSpring=true
    bOverrideAngularDamping=true
    bUseGravityOverride=true
    simulationSpace=BoneRelative
    alphaBoolBlend {blendOption=Linear}
    boundBone {boneName=dyn_antena_01}
    componentAppliedLinearAccClamp=<200 200 200>
    gravityOverride {z=-980}
    physicsBodyDefinitions [{
      boundBone {boneName=dyn_antena_01}
      boxExtents=<10 10 10>
      constraintSetup {angularLimitsMax=<0 5 0>, angularLimitsMin=<0 -5 0>, angularTarget=<1 0 0>}
      localJointOffset=<0.05 0 0>
     }]
   }
   n2 = Anim Dynamics - Bone: dyn_antena_02(ComponentPose=@n1, Alpha=1, ExternalForce="0,0,0") {
    angularDampingOverride=1
    angularSpringConstant=3000
    bAngularSpring=true
    bOverrideAngularDamping=true
    bUseGravityOverride=true
    simulationSpace=World
    alphaBoolBlend {blendOption=Linear}
    boundBone {boneName=dyn_antena_02}
    componentAppliedLinearAccClamp=<200 200 200>
    gravityOverride {z=-980}
    physicsBodyDefinitions [{
      boundBone {boneName=dyn_antena_02}
      boxExtents=<10 10 10>
      constraintSetup {angularLimitsMax=<0 4 0>, angularLimitsMin=<0 -4 0>, angularTarget=<1 0 0>}
      localJointOffset=<0.05 0 0>
     }]
   }
   n3 = Component To Local(ComponentPose=@n2)
   n4 = Anim Dynamics - Bone: dyn_antena_01(ComponentPose=Local To Component(LocalPose=Use cached pose 'Init'()), Alpha=1, ExternalForce="0,0,0") {
    angularDampingOverride=1
    angularSpringConstant=4500
    bAngularSpring=true
    bOverrideAngularDamping=true
    bUseGravityOverride=true
    simulationSpace=RootRelative
    alphaBoolBlend {blendOption=Linear}
    boundBone {boneName=dyn_antena_01}
    componentAppliedLinearAccClamp=<3000 3000 3000>
    componentLinearAccScale=<0.2 0.2 0.2>
    componentLinearVelScale=<0.2 0.2 0.2>
    gravityOverride {z=-980}
    physicsBodyDefinitions [{
      boundBone {boneName=dyn_antena_01}
      boxExtents=<15 15 15>
      constraintSetup {angularLimitsMax=<0 5 5>, angularLimitsMin=<0 -5 -5>, angularTarget=<1 0 0>}
      localJointOffset=<0.05 0 0>
     }]
   }
   n5 = Anim Dynamics - Bone: dyn_antena_02(ComponentPose=@n4, Alpha=1, ExternalForce="0,0,0") {
    angularDampingOverride=1
    angularSpringConstant=4500
    bAngularSpring=true
    bOverrideAngularDamping=true
    bUseGravityOverride=true
    simulationSpace=RootRelative
    alphaBoolBlend {blendOption=Linear}
    boundBone {boneName=dyn_antena_02}
    componentAppliedLinearAccClamp=<3000 3000 3000>
    componentLinearAccScale=<0.2 0.2 0.2>
    componentLinearVelScale=<0.2 0.2 0.2>
    gravityOverride {z=-980}
    physicsBodyDefinitions [{
      boundBone {boneName=dyn_antena_02}
      boxExtents=<15 15 15>
      constraintSetup {angularLimitsMax=<0 5 5>, angularLimitsMin=<0 -5 -5>, angularTarget=<1 0 0>}
      localJointOffset=<0.025 0 0>
     }]
   }
   n6 = Anim Dynamics - Bone: dyn_antena_03(ComponentPose=@n5, Alpha=1, ExternalForce="0,0,0") {
    angularDampingOverride=1
    angularSpringConstant=4500
    bAngularSpring=true
    bOverrideAngularDamping=true
    bUseGravityOverride=true
    simulationSpace=RootRelative
    alphaBoolBlend {blendOption=Linear}
    boundBone {boneName=dyn_antena_03}
    componentAppliedLinearAccClamp=<3000 3000 3000>
    componentLinearAccScale=<0.2 0.2 0.2>
    componentLinearVelScale=<0.2 0.2 0.2>
    gravityOverride {z=-980}
    physicsBodyDefinitions [{
      boundBone {boneName=dyn_antena_03}
      boxExtents=<15 15 15>
      constraintSetup {angularLimitsMax=<0 5 5>, angularLimitsMin=<0 -5 -5>, angularTarget=<1 0 0>}
      localJointOffset=<0.025 0 0>
     }]
   }
   n7 = "Transform (Modify) Bone - Bone: constr_Antena_jnt"!(ComponentPose=@n6, Alpha=1, Rotation="15,0,0") -> (Pose) {rotationMode=BMM_Additive, alphaBoolBlend {blendOption=Linear}, boneToModify {boneName=constr_Antena_jnt}, rotation {pitch=15}}
   n8 = Component To Local(ComponentPose=@n7)
   n9 = RigidBody!(
    ComponentPose=Local To Component(LocalPose=Blend Poses by bool(BlendPose_0=@n3, BlendPose_1=@n8) {blendType=Linear})
    Alpha
    OverrideWorldGravity:Vector="0,0,0"
    ExternalForce
   ) -> (Pose) {
    bClampLinearTranslationLimitToRefPose=true
    bOverrideWorldGravity=true
    overridePhysicsAsset=&/M_MED_ArcticBreeze_Physics
    simulationSpace=BaseBoneSpace
    alphaBoolBlend {blendOption=Linear}
    baseBoneRef {boneName=pelvis}
    componentAppliedLinearAccClamp=<3000 3000 3000>
    simSpaceSettings {
     dampingAlpha=0
     maxAngularAcceleration=100
     maxAngularVelocity=6
     maxLinearAcceleration=1200
     maxLinearVelocity=500
     worldAlpha=1
     externalLinearDragV=<0.5 0.5 0.5>
    }
   }
   n10 = Anim Dynamics - Bone: dyn_radioWire_01!(ComponentPose=@n9, bAlphaBoolEnabled:bool=True, ExternalForce="0,0,0") -> (Pose) {
    alphaInputType=Bool
    angularDampingOverride=0.9999
    angularSpringConstant=300
    bAngularSpring=true
    bLinearSpring=true
    bOverrideAngularDamping=true
    bOverrideLinearDamping=true
    bUseGravityOverride=true
    linearDampingOverride=0.9999
    linearSpringConstant=150
    simulationSpace=RootRelative
    alphaBoolBlend {blendOption=Linear}
    boundBone {boneName=dyn_radioWire_01}
    componentAppliedLinearAccClamp=<500 500 500>
    componentLinearAccScale=<0.1 0.1 0.1>
    componentLinearVelScale=<0.1 0.1 0.1>
    gravityOverride {z=-980}
    physicsBodyDefinitions [{
      boundBone {boneName=dyn_radioWire_01}
      boxExtents=<10 10 10>
      constraintSetup {angularLimitsMax=<18 3 40>, angularLimitsMin=<-18 -3 -40>, angularTarget=<1 0 0>, linearAxesMax=<0.7 0.7 0.7>, linearAxesMin=<-0.7 -0.7 0>}
      localJointOffset=<0.5 0 0>
     }]
    relativeSpaceBone {boneName=dyn_radio_03}
   }
   n11 = Anim Dynamics - Bone: dyn_radioWire_01(ComponentPose=@n10, bAlphaBoolEnabled=True, ExternalForce="0,0,0") {
    alphaInputType=Bool
    angularDampingOverride=0.9999
    angularSpringConstant=300
    bAngularSpring=true
    bLinearSpring=true
    bOverrideAngularDamping=true
    bOverrideLinearDamping=true
    bUseGravityOverride=true
    linearDampingOverride=0.9999
    linearSpringConstant=150
    simulationSpace=World
    alphaBoolBlend {blendOption=Linear}
    boundBone {boneName=dyn_radioWire_01}
    componentAppliedLinearAccClamp=<500 500 500>
    componentLinearAccScale=<0.1 0.1 0.1>
    componentLinearVelScale=<0.1 0.1 0.1>
    gravityOverride {z=-980}
    physicsBodyDefinitions [{
      boundBone {boneName=dyn_radioWire_01}
      boxExtents=<10 10 10>
      constraintSetup {angularLimitsMax=<18 3 40>, angularLimitsMin=<-18 -3 -40>, angularTarget=<1 0 0>, linearAxesMax=<0.7 0.7 0.7>, linearAxesMin=<-0.7 -0.7 0>}
      localJointOffset=<0.5 0 0>
     }]
    relativeSpaceBone {boneName=dyn_radio_03}
   }
   Output Pose(Result=Component To Local(ComponentPose=@n11)) {layerGroup=None}
   n12 = "Transform (Modify) Bone - Bone: attach_cape"!(ComponentPose=Local To Component(LocalPose=Copy Pose From Mesh() {bUseAttachedParent=true}), Alpha=1, Translation="0,1.7,0") -> (Pose) {
    rotationMode=BMM_Additive
    translationMode=BMM_Additive
    translationSpace=BCS_ParentBoneSpace
    alphaBoolBlend {blendOption=Linear}
    boneToModify {boneName=attach_cape}
    translation {y=1.7}
   }
   n13 = "Transform (Modify) Bone - Bone: attach_backpack"!(ComponentPose=@n12, Alpha=1, Translation="0,0.7,0") -> (Pose) {
    rotationMode=BMM_Additive
    translationMode=BMM_Additive
    alphaBoolBlend {blendOption=Linear}
    boneToModify {boneName=attach_backpack}
    translation {y=0.7}
   }
   n14 = Control Rig!(Source=Component To Local(ComponentPose=@n13), Alpha, isDBNO:bool=Get bIsDBNO()) -> (Pose) {
    alphaBoolBlend {blendOption=Linear}
    controlRigAssetReference {blueprintRigClass=&/M_MED_ArcticBreeze_CtrlRig.M_MED_ArcticBreeze_CtrlRig_C}
   }
   n15 = "Constraint - constr_Antena_jnt (PPPPPPPP)"!(
    ComponentPose=Local To Component(LocalPose=@n14)
    Alpha=1
    ConstraintWeights_0=0.6322
    ConstraintWeights_1:float=0.1815
    ConstraintWeights_2:float=0.1055
    ConstraintWeights_3:float=0.0122
    ConstraintWeights_4:float=0.0055
    ConstraintWeights_5:float=0.0115
    ConstraintWeights_6:float=0.0012
    ConstraintWeights_7:float=0.0503
   ) -> (Pose) {
    alphaBoolBlend {blendOption=Linear}
    boneToModify {boneName=constr_Antena_jnt}
    constraintSetup [
     {targetBone {boneName=spine_04}}
     {targetBone {boneName=spine_05}}
     {targetBone {boneName=clavicle_l}}
     {targetBone {boneName=upperarm_twist_01_l}}
     {targetBone {boneName=deform_upperarm_fr_l}}
     {targetBone {boneName=deform_upperarm_up_l}}
     {targetBone {boneName=neck_02}}
     {targetBone {boneName=deform_pec_l}}
    ]
    constraintWeights [1,1,1,1,1,1,1,1]
   }
   n16 = "Constraint - constr_Holster_jnt (PPPPPPPP)"!(
    ComponentPose=@n15
    Alpha=1
    ConstraintWeights_0=0.0205
    ConstraintWeights_1:float=0.1444
    ConstraintWeights_2:float=0.3456
    ConstraintWeights_3:float=0.4639
    ConstraintWeights_4:float=0.0004
    ConstraintWeights_5:float=0
    ConstraintWeights_6:float=0
    ConstraintWeights_7:float=0.0253
   ) -> (Pose) {
    alphaBoolBlend {blendOption=Linear}
    boneToModify {boneName=constr_Holster_jnt}
    constraintSetup [
     {targetBone {boneName=spine_01}}
     {targetBone {boneName=spine_02}}
     {targetBone {boneName=spine_03}}
     {targetBone {boneName=spine_04}}
     {targetBone {boneName=upperarm_twist_01_l}}
     {targetBone {boneName=deform_upperarm_bk_l}}
     {targetBone {boneName=upperarm_twist_02_l}}
     {targetBone {boneName=deform_pec_l}}
    ]
    constraintWeights [1,1,1,1,1,1,1,1]
   }
   n17 = "Constraint - constr_Knife_jnt (PPP)"!(ComponentPose=@n16, Alpha=1, ConstraintWeights_0=0.2068, ConstraintWeights_1:float=0.0636, ConstraintWeights_2:float=0.7296) -> (Pose) {
    alphaBoolBlend {blendOption=Linear}
    boneToModify {boneName=constr_Knife_jnt}
    constraintSetup [{targetBone {boneName=calf_r}}, {targetBone {boneName=calf_twist_01_r}}, {targetBone {boneName=calf_twist_02_r}}]
    constraintWeights [1,1,1]
   }
   n18 = "Constraint - constr_MagR_jnt (PPPPPP)"!(
    ComponentPose=@n17
    Alpha=1
    ConstraintWeights_0=0.2397
    ConstraintWeights_1:float=0.2036
    ConstraintWeights_2:float=0.3018
    ConstraintWeights_3:float=0.2422
    ConstraintWeights_4:float=0.0126
    ConstraintWeights_5:float=0
   ) -> (Pose) {
    alphaBoolBlend {blendOption=Linear}
    boneToModify {boneName=constr_MagR_jnt}
    constraintSetup [
     {targetBone {boneName=pelvis}}
     {targetBone {boneName=spine_01}}
     {targetBone {boneName=spine_02}}
     {targetBone {boneName=spine_03}}
     {targetBone {boneName=spine_04}}
     {targetBone {boneName=deform_pec_r}}
    ]
    constraintWeights [1,1,1,1,1,1]
   }
   n19 = "Constraint - constr_MagMid_jnt (PPPPPPPP)"!(
    ComponentPose=@n18
    Alpha=1
    ConstraintWeights_0=0.2555
    ConstraintWeights_1:float=0.1755
    ConstraintWeights_2:float=0.3197
    ConstraintWeights_3:float=0.227
    ConstraintWeights_4:float=0.0221
    ConstraintWeights_5:float=0
    ConstraintWeights_6:float=0.0001
    ConstraintWeights_7:float=0.0001
   ) -> (Pose) {
    alphaBoolBlend {blendOption=Linear}
    boneToModify {boneName=constr_MagMid_jnt}
    constraintSetup [
     {targetBone {boneName=pelvis}}
     {targetBone {boneName=spine_01}}
     {targetBone {boneName=spine_02}}
     {targetBone {boneName=spine_03}}
     {targetBone {boneName=spine_04}}
     {targetBone {boneName=spine_05}}
     {targetBone {boneName=deform_pec_r}}
     {targetBone {boneName=deform_pec_l}}
    ]
    constraintWeights [1,1,1,1,1,1,1,1]
   }
   n20 = "Constraint - constr_MagL_jnt (PPPPPP)"!(
    ComponentPose=@n19
    Alpha=1
    ConstraintWeights_0=0.2104
    ConstraintWeights_1:float=0.2017
    ConstraintWeights_2:float=0.3081
    ConstraintWeights_3:float=0.2655
    ConstraintWeights_4:float=0.0143
    ConstraintWeights_5:float=0
   ) -> (Pose) {
    alphaBoolBlend {blendOption=Linear}
    boneToModify {boneName=constr_MagL_jnt}
    constraintSetup [
     {targetBone {boneName=pelvis}}
     {targetBone {boneName=spine_01}}
     {targetBone {boneName=spine_02}}
     {targetBone {boneName=spine_03}}
     {targetBone {boneName=spine_04}}
     {targetBone {boneName=deform_pec_l}}
    ]
    constraintWeights [1,1,1,1,1,1]
   }
   n21 = "Constraint - dyn_frontBelt_master (PPPP)"!(
    ComponentPose=@n20
    Alpha=1
    ConstraintWeights_0=0.7981
    ConstraintWeights_1:float=0.0708
    ConstraintWeights_2:float=0.0823
    ConstraintWeights_3:float=0.0488
   ) -> (Pose) {
    alphaBoolBlend {blendOption=Linear}
    boneToModify {boneName=dyn_frontBelt_master}
    constraintSetup [{targetBone {boneName=pelvis}}, {targetBone {boneName=spine_01}}, {targetBone {boneName=spine_02}}, {targetBone {boneName=spine_03}}]
    constraintWeights [1,1,1,1]
   }
   n22 = "Constraint - dyn_backBelt_master (PPPP)"!(
    ComponentPose=@n21
    Alpha=1
    ConstraintWeights_0=0.8258
    ConstraintWeights_1:float=0.0647
    ConstraintWeights_2:float=0.0656
    ConstraintWeights_3:float=0.0438
   ) -> (Pose) {
    alphaBoolBlend {blendOption=Linear}
    boneToModify {boneName=dyn_backBelt_master}
    constraintSetup [{targetBone {boneName=pelvis}}, {targetBone {boneName=spine_01}}, {targetBone {boneName=spine_02}}, {targetBone {boneName=spine_03}}]
    constraintWeights [1,1,1,1]
   }
   n23 = "Transform (Modify) Bone - Bone: constr_MagL_jnt"!(ComponentPose=@n22, Alpha=1, Rotation="0,0,20") -> (Pose) {rotationMode=BMM_Additive, alphaBoolBlend {blendOption=Linear}, boneToModify {boneName=constr_MagL_jnt}, rotation {roll=20}}
   n24 = "Transform (Modify) Bone - Bone: constr_MagR_jnt"!(ComponentPose=@n23, Alpha=1, Rotation="0,0,20") -> (Pose) {rotationMode=BMM_Additive, alphaBoolBlend {blendOption=Linear}, boneToModify {boneName=constr_MagR_jnt}, rotation {roll=20}}
   n25 = "Transform (Modify) Bone - Bone: constr_MagMid_jnt"!(ComponentPose=@n24, Alpha=1, Rotation="0,0,20") -> (Pose) {rotationMode=BMM_Additive, alphaBoolBlend {blendOption=Linear}, boneToModify {boneName=constr_MagMid_jnt}, rotation {roll=20}}
   n26 = "Transform (Modify) Bone - Bone: constr_MagL_jnt"!(ComponentPose=@n25, Alpha=1, Translation:Vector="0.000,0.000,1.234", Rotation="0,0,0") -> (Pose) {translationMode=BMM_Additive, alphaBoolBlend {blendOption=Linear}, boneToModify {boneName=constr_MagL_jnt}, translation {z=1.234}}
   n27 = "Transform (Modify) Bone - Bone: constr_MagMid_jnt"!(ComponentPose=@n26, Alpha=1, Translation:Vector="0,0,0.6", Rotation="0,0,0") -> (Pose) {translationMode=BMM_Additive, alphaBoolBlend {blendOption=Linear}, boneToModify {boneName=constr_MagMid_jnt}, translation {z=0.6}}
   Save cached pose 'Init'(Pose=Component To Local(ComponentPose=@n27)) {cacheName=Init}
  ]}
 EventGraph {nodes [
   Event BlueprintUpdateAnimation()
   > n1 = Sequence()
   > n2 = Gravity Override State Function DEPRECATED(CharacterPartAnimInstance=Self-Reference(), GameTimeInSeconds=GetGameTimeInSeconds(), GravityOverrideParams=Get Gravity Override Params())
   > Set GravityOverride(GravityOverride=@n2)
   Set Is NOT In Front End(execute=@n1.then_1, Is NOT In Front End:bool=NOT Boolean(A=Get bIsInFrontEnd())) -> (Output_Get:bool)
   > Set SpineBendFactor(SpineBendFactor:double=GetCurveValue(CurveName=SpineBendFactor)) -> (Output_Get:double)
   > Set LeftThighUpAlpha(LeftThighUpAlpha:float=GetCurveValue(CurveName=LeftThighUpAlpha)) -> (Output_Get:float)
   > Set AntenaDynamicsAlpha(AntenaDynamicsAlpha:float=GetCurveValue(CurveName=AntenaDynamicsAlpha)) -> (Output_Get:float)
   > Set AntenaBaseCorrAlpha(AntenaBaseCorrAlpha:float=GetCurveValue(CurveName=AntenaBaseCorrAlpha)) -> (Output_Get:float)
  ]}
]
variables [
 AntenaBaseCorrAlpha {type=number}
 AntenaDynamicsAlpha {type=number}
 Gravity Override Params {type=GravityOverrideParamsStruct}
 GravityOverride {type=Vector}
 Is NOT In Front End {type=boolean}
 LeftThighUpAlpha {type=number}
 SpineBendFactor {type=number}
]
